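# Unlit masked Niagara particle material for doughnut chunk effects, with dithered masking and support for sprites, meshes and ribbons.
# Fakes shading without lights by transforming a chunk normal map, remapping its X and Y into UVs for a tile pattern texture, then smoothstepping the result into brightness.
# That brightness multiplies the chunk texture's blue-tinted red and brown-tinted blue channels, then particle color, feeding an emissive adjustment function.
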Material /DealershipSharedLibrary/Assets/Sets/Doughnuts/Material/M_Doughnut_Chunks
graph {
 properties {
  bAutomaticallySetUsageInEditor=false
  bUsedWithNiagaraMeshParticles=true
  bUsedWithNiagaraRibbons=true
  bUsedWithNiagaraSprites=true
  blendMode=BLEND_Masked
  ditherOpacityMask=true
  opacityMaskClipValue=0.33
  shadingModel=MSM_Unlit
  usedShadingModels=Unlit
 }
 nodes [
  n1 = MaterialFunctionCall(Color=Transform(None=TextureSample() {samplerType=SAMPLERTYPE_Normal, texture=/DealershipSharedLibrary/Assets/Sets/Doughnuts/Textures/T_VCVFX_Chunks_N}.RGB) {transformType=TRANSFORM_Local}) {materialFunction=&SplitComponents}
  n2 = Add(A=Constant() {r=1}, B=@n1.R)
  n3 = Constant() {r=0.25}
  n4 = Multiply(A=@n3, B=Add(A=@n1.G, B=Constant() {r=1}))
  n5 = Constant() {r=1}
  n6 = TextureSample(UVs=AppendVector(A=Multiply(A=@n2, B=@n3), B=Multiply(A=@n4, B=@n5))) {texture=/Game/Effects/Fort_Effects/Textures/T_Tile_Pattern_01}
  n7 = Multiply(A=Constant() {r=3}, B=Add(A=Constant() {r=0.35}, B=SmoothStep(Value=@n6.RGB) {constMax=0.5}))
  n8 = TextureSample() {texture=/DealershipSharedLibrary/Assets/Sets/Doughnuts/Textures/T_VCVFX_Chunks}
  n9 = Add(
   A=Multiply(A=@n8.R, B=Constant3Vector() {constant=<rgba 0.0801 0.6468 1 1>})
   B=Multiply(A=@n8.B, B=Constant3Vector() {constant=<rgba 0.6615 0.2264 0.1127 1>})
  )
  MaterialFunctionCall(Emissive=Multiply(A=Multiply(A=@n7, B=@n9), B=ParticleColor().RGB)) {materialFunction=&MF_FX_EmissiveAdjustments}
  Reroute(None=@n8.A)
 ]
}
