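# Unlit, masked, dithered particle material for the Doughnuts set that fakes lighting on sprinkle textures, enabled for Niagara sprites, meshes and ribbons.
# A transformed normal map's offset, scaled X and Y become UVs into a tile pattern, which is smoothstepped into light and shadow with adjustable intensity.
# That result multiplies particle color and a selected texture channel, then passes through emissive adjustments.
# Another selectable channel of the same texture supplies opacity, alpha by default.
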Material /DealershipSharedLibrary/Assets/Sets/Doughnuts/Material/M_Doughnut_FakeLight
graph {
 properties {
  bAutomaticallySetUsageInEditor=false
  bUsedWithNiagaraMeshParticles=true
  bUsedWithNiagaraRibbons=true
  bUsedWithNiagaraSprites=true
  blendMode=BLEND_Masked
  ditherOpacityMask=true
  opacityMaskClipValue=0.33
  shadingModel=MSM_Unlit
  usedShadingModels=Unlit
 }
 nodes [
  n1 = TextureSampleParameter2D() {
   parameterName=Normal Texture Fake Light
   samplerType=SAMPLERTYPE_Normal
   texture=/DealershipSharedLibrary/Assets/Sets/Doughnuts/Textures/T_VCVFX_Sprinkles_N
  }
  n2 = MaterialFunctionCall(Color=Transform(None=@n1.RGB) {transformType=TRANSFORM_Local}) {materialFunction=&SplitComponents}
  n3 = Add(A=Constant() {r=1}, B=@n2.R)
  n4 = ScalarParameter() {defaultValue=0.1, parameterName=Fake Light Break Amount}
  n5 = Multiply(A=@n4, B=Add(A=@n2.G, B=Constant() {r=1}))
  n6 = Constant() {r=1}
  n7 = TextureSample(UVs=AppendVector(A=Multiply(A=@n3, B=@n4), B=Multiply(A=@n5, B=@n6))) {texture=/Game/Effects/Fort_Effects/Textures/T_Tile_Pattern_01}
  n8 = Multiply(
   A=ScalarParameter() {defaultValue=3, parameterName=Fake Light Intensity}
   B=Add(A=ScalarParameter() {defaultValue=0.4, parameterName=Fake Shadow Strength}, B=SmoothStep(Value=@n7.RGB) {constMax=0.5})
  )
  n9 = ParticleColor()
  n10 = TextureSampleParameter2D() {parameterName=Texture, texture=/DealershipSharedLibrary/Assets/Sets/Doughnuts/Textures/T_VCVFX_Sprinkles}
  MaterialFunctionCall(Emissive=Multiply(A=Multiply(A=@n8, B=@n9.RGB), B=ChannelMaskParameter(None=@n10.RGB) {parameterName=Color Channel}.R)) {materialFunction=&MF_FX_EmissiveAdjustments}
  Reroute(None=ChannelMaskParameter(None=@n10.RGBA) {maskChannel=Alpha, parameterName=Opacity Channel, defaultValue=<rgba 0 0 0 1>}.R)
 ]
}
