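# Masked, lit material for doughnut meshes rendered as Niagara mesh particles, with base colour, normal, specular and packed-mask textures.
# Particle Color tints the doughnut texture through lerps that use a TintBlend parameter, with named reroutes for frosting and bread masks and a desaturated sprinkle tint.
# Dynamic parameters scale bread brightness and drive a dissolve that thresholds a tiled texture through dithered temporal AA.
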Material /DealershipSharedLibrary/Assets/Sets/Doughnuts/Meshes/Material/M_SimpleBase_Doughnut_Mesh
graph {
 properties {
  bAutomaticallySetUsageInEditor=false
  bUsedWithNiagaraMeshParticles=true
  blendMode=BLEND_Masked
  previewMesh=/DealershipSharedLibrary/Assets/Sets/Doughnuts/Meshes/SM_Donut_Half
  usedShadingModels=Default Lit
 }
 nodes [
  Constant()
  n1 = ParticleColor()
  n2 = LinearInterpolate(A=NamedRerouteUsage(), B=Constant() {r=0.4}, Alpha=NamedRerouteUsage())
  n3 = NamedRerouteUsage()
  n4 = ParticleColor()
  n5 = NamedRerouteUsage()
  LinearInterpolate(
   A=LinearInterpolate(A=Multiply(A=@n1.RGB, B=@n2), B=LinearInterpolate(A=@n3, B=@n4.RGB, Alpha=@n5), Alpha=ScalarParameter() {parameterName=TintBlend})
   B=Multiply(A=NamedRerouteUsage(), B=DynamicParameter() {paramNames [Dissolve, BreadBrightnes, Param3, Param4]}.BreadBrightnes)
   Alpha=NamedRerouteUsage()
  )
  n6 = TextureSampleParameter2D() {parameterName=Specular, texture=/DealershipSharedLibrary/Assets/Sets/Doughnuts/Meshes/Texture/T_Donut_S}
  LinearInterpolate(A=@n6.B, Alpha=NamedRerouteUsage())
  Multiply(A=@n6.RGB, B=Constant() {r=1.4})
  NamedRerouteDeclaration(None=TextureSampleParameter2D() {parameterName=BC, texture=/DealershipSharedLibrary/Assets/Sets/Doughnuts/Meshes/Texture/T_Donut_D}.RGB) {name=Donut Color, nodeColor=<rgba 1 0.0234 0 1>}
  n7 = Desaturation(None=NamedRerouteUsage(), Fraction=Constant() {r=1})
  NamedRerouteDeclaration(None=Multiply(A=Multiply(A=@n7, B=@n7), B=Constant() {r=4})) {name=Sprinkles Tint, nodeColor=<rgba 0 0.2969 1 1>}
  n8 = TextureSampleParameter2D() {parameterName=Mask, texture=/DealershipSharedLibrary/Assets/Sets/Doughnuts/Meshes/Texture/T_Donut_Mask_RGBPacked}
  NamedRerouteDeclaration(None=@n8.R) {name=Frosting Mask, nodeColor=<rgba 0.6172 1 0 1>}
  NamedRerouteDeclaration(None=@n8.B) {name=Bread Mask, nodeColor=<rgba 0 1 0.25 1>}
  n9 = TextureSample() {texture=/Game/Effects/Fort_Effects/Textures/Tile/T_Smooth_Tile_Chromatic}
  n10 = Subtract(A=Add(A=OneMinus(None=@n9.R)), B=Multiply(A=DynamicParameter() {paramNames [Dissolve, BreadBrightnes, Param3, Param4]}.Dissolve) {constB=2})
  Reroute(None=MaterialFunctionCall(Alpha Threshold=Saturate(None=@n10)) {materialFunction=&DitherTemporalAA_Fortnite}.Result)
  TextureSampleParameter2D() {parameterName=N, samplerType=SAMPLERTYPE_Normal, texture=/DealershipSharedLibrary/Assets/Sets/Doughnuts/Meshes/Texture/T_Donut_N}
 ]
}
